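# Niagara system for the first-attack swing effect of the CreamDecoy mic stand pickaxe cosmetic.
# Three mesh emitters (Air_cone, AirflowColor, Airflow_Stick) burst after slight delays at sampled User Skeletal Mesh locations, drawing short-lived air streaks animated by scale and dynamic material curves.
# They start from System TOD Color, and System VariantInt switches between bright and dark color and alpha fades.
# A sprite Sparks emitter is present but disabled.
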NiagaraSystem /SparksCosmetics/Instrument/MicStand/CreamDecoy/Effects/NS_Pickaxe_MicStand_CreamDecoy_FirstAttack
dependencies {
 usedDataInterfaces [&NiagaraDataInterfaceCurve, &NiagaraDataInterfaceColorCurve, &NiagaraDataInterfaceVector2DCurve]
 usedDynamicInputs [&MakeIntFromFloat]
 usedModules [
  &self:SystemSpawnScript.SystemScriptSource.SystemScriptGraph.NiagaraNodeAssignment_1.SetVariables_#1
  &SystemState
 ]
}
emitters {
 Air_cone {
  emitterUpdateScript [
   &EmitterState(Life Cycle Mode=::NewEnumerator1, Loop Behavior=::NewEnumerator1, Loop Delay=0.08, Loop Duration=0.5, UseLoopDelay=true)
   &SpawnBurst_Instantaneous(Use Spawn Probability=false, UseLoopCountLimit=false)
  ]
  particleSpawnScript [
   &InitializeParticle(
    Color=$System.TOD Color
    Lifetime=0.2
    Mesh Scale=&Multiply_VectorByFloat(Vector=<0.42 0.1 0.1>, Float=1)
    Mesh Scale Mode=::NewEnumerator4
    UsePositionOffset=false
    Write Material Random=true
   )
   &InitialMeshOrientation(Mesh Orientation Mode=::NewEnumerator2, Rotation=<0 0.25 0>, Use Rotation Vector=true)
   &SkeletalMeshLocation(
    MaskSpawnGroup=false
    Mesh Sampling Type=::NewEnumerator5
    Sampled Position Offset=<0 0 -20>
    Skeletal Mesh=$User.Skeletal Mesh
    Socket Sampling Mode=::NewEnumerator5
   )
  ]
  particleUpdateScript [
   &ParticleState()
   &ScaleMeshSize(Scale Factor=&MakeVector(
     X=&FloatFromCurve(
      FloatCurve=DataInterface() {propertyValues {Curve {keys [{arriveTangent=0.286, interpMode=RCIM_Cubic, leaveTangent=0.286, tangentMode=RCTM_User, time=0, value=1}, <at 1 1.2>]}}}
      Scale Curve=0.65
     )
     Y=&FloatFromCurve(
      FloatCurve=DataInterface() {propertyValues {Curve {keys [{arriveTangent=3.1084, interpMode=RCIM_Cubic, leaveTangent=3.1084, tangentMode=RCTM_User, time=0, value=0.5}, <at 0.3 1>]}}}
      Scale Curve=2
     )
     Z=&FloatFromCurve(
      FloatCurve=DataInterface() {propertyValues {Curve {keys [{arriveTangent=3.1084, interpMode=RCIM_Cubic, leaveTangent=3.1084, tangentMode=RCTM_User, time=0, value=0.5}, <at 0.3 1>]}}}
      Scale Curve=2
     )
    ))
   &DynamicMaterialParameters(
    Index 0 Param 1=0.25
    Index 0 Param 2=0.5
    Index 0 Param 3=1
    Index 0 Param 4=&FloatFromCurve(FloatCurve=DataInterface() {propertyValues {Curve {keys [{arriveTangent=-1.3436, interpMode=RCIM_Cubic, leaveTangent=-1.3436, tangentMode=RCTM_User, time=0, value=0.75}, <at 1 0>]}}})
    Index 3 Param 2=1
    Module.Param0WriteEnabled=true
    Module.Param0WriteEnabled3=true
    Module.Param1WriteEnabled=true
    Module.Param1WriteEnabled3=true
    Module.Param2WriteEnabled=true
    Module.Param2WriteEnabled3=true
    Module.Param3WriteEnabled=true
    Module.Param3WriteEnabled3=true
    Write Parameter Index 3=true
   ) {module=DynamicMaterialParameters001}
   &ScaleColor(
    Scale Alpha=&MergeHelper_FloatSwitch(
     0=&FloatFromCurve(FloatCurve=DataInterface() {propertyValues {Curve {keys [<at 0 1>,<at 1 0>]}}}, Scale Curve=0.2)
     1=&FloatFromCurve(
      FloatCurve=DataInterface() {propertyValues {Curve {keys [<at 0 1 RCIM_Cubic>, {arriveTangent=-0.501, interpMode=RCIM_Cubic, leaveTangent=-0.501, tangentMode=RCTM_User, time=1, value=0}]}}}
      Scale Curve=1.25
     )
     2=?
     3=?
     Index=$System.VariantInt
     Num=::NewEnumerator0
    )
    Scale RGB=&MergeHelper_VectorSwitch(
     0=&MakeVectorFromLinearColorRGB(Color=&ColorFromCurve(ColorCurve=DataInterface() {propertyValues {
         AlphaCurve {keys []}
         BlueCurve {keys [<at 0 2 RCIM_Cubic>,<at 0.4076 0.5553 RCIM_Cubic>]}
         GreenCurve {keys [<at 0 2 RCIM_Cubic>,<at 0.4076 0.5553 RCIM_Cubic>]}
         RedCurve {keys [<at 0 2 RCIM_Cubic>,<at 0.4076 0.5553 RCIM_Cubic>]}
        }}))
     1=&MakeVectorFromLinearColorRGB(Color=&ColorFromCurve(ColorCurve=DataInterface() {propertyValues {
         AlphaCurve {keys []}
         BlueCurve {keys [<at 0 0.1745>,<at 1 0.0595 RCIM_Cubic>]}
         GreenCurve {keys [<at 0 0.113>,<at 1 0.0185 RCIM_Cubic>]}
         RedCurve {keys [<at 0 0.1022>,<at 1 0.0123 RCIM_Cubic>]}
        }}))
     2=?
     3=?
     Index=$System.VariantInt
     Num=::NewEnumerator0
    )
    ScaleA=true
    ScaleRGB=true
   )
   &FN_Mesh_Attributes(
    Color=&MakeLinearColorFromVectorAndFloat(
     "Vector (RGB)"=&Multiply_Vector(A=&MakeVectorFromLinearColorRGB(Color=$System.TOD Color), B=<1 1 1>)
     "Float (Alpha)"=&FloatFromCurve(
      FloatCurve=DataInterface() {propertyValues {Curve {keys [{arriveTangent=-0.2188, interpMode=RCIM_Cubic, leaveTangent=-0.2188, tangentMode=RCTM_User, time=0, value=1}, <at 1 0>]}}}
      Scale Curve=2
     )
    )
    Create Color Attribute=true
    Create Scale Attribute=true
    Module Output=::NewEnumerator1
    Scale=&Multiply_Vector(
     A=<0.12 0.25 0.25>
     B=&MakeVector(
      X=&FloatFromCurve(FloatCurve=DataInterface() {propertyValues {Curve {keys [{arriveTangent=0.3763, interpMode=RCIM_Cubic, leaveTangent=0.3763, tangentMode=RCTM_User, time=0, value=1}, <at 1 1.2>]}}})
      Y=&FloatFromCurve(FloatCurve=DataInterface() {propertyValues {Curve {keys [{arriveTangent=3.1084, interpMode=RCIM_Cubic, leaveTangent=3.1084, tangentMode=RCTM_User, time=0, value=0.5}, <at 0.3 1>]}}})
      Z=&FloatFromCurve(FloatCurve=DataInterface() {propertyValues {Curve {keys [{arriveTangent=3.1084, interpMode=RCIM_Cubic, leaveTangent=3.1084, tangentMode=RCTM_User, time=0, value=0.5}, <at 0.3 1>]}}})
     )
    )
   ) {module=FN_Mesh_Attributes001}
   /SetVariables(Particles.Dy3:Vector4f=&MakeVector4(X=0, Y=1, Z=0, W=1)) {isSetParameters=true}
  ]
  rendererClasses [&NiagaraMeshRendererProperties]
  renderers [
   {rendererClass=&NiagaraMeshRendererProperties, rendererIndex=0}
   {rendererClass=&NiagaraMeshRendererProperties, rendererIndex=1}
   {rendererClass=&NiagaraMeshRendererProperties, rendererIndex=2}
  ]
 }
 AirflowColor {
  emitterUpdateScript [
   &EmitterState(Life Cycle Mode=::NewEnumerator1, Loop Behavior=::NewEnumerator1, Loop Delay=0.07, Loop Duration=0.5, UseLoopDelay=true)
   &SpawnBurst_Instantaneous(Use Spawn Probability=false, UseLoopCountLimit=false)
  ]
  particleSpawnScript [
   &InitializeParticle(
    Color=$System.TOD Color
    Lifetime=0.25
    Mesh Scale=<0.85 0.8 0.8>
    Mesh Scale Mode=::NewEnumerator4
    UsePositionOffset=false
    Write Material Random=true
   )
   &InitialMeshOrientation(Mesh Orientation Mode=::NewEnumerator2, Rotation=<0 -0.25 0>, Use Rotation Vector=true)
   &SkeletalMeshLocation(
    MaskSpawnGroup=false
    Mesh Sampling Type=::NewEnumerator5
    Orientation Sampling=::NewEnumerator1
    Sampled Position Offset=<0 0 80>
    Skeletal Mesh=$User.Skeletal Mesh
    Socket Sampling Mode=::NewEnumerator5
   )
   &CameraOffset(Camera Offset Amount=20)
  ]
  particleUpdateScript [
   &ParticleState()
   &ScaleMeshSize(Scale Factor=&MakeVector(
     X=&FloatFromCurve(FloatCurve=DataInterface() {propertyValues {Curve {keys [<at 0 1>,<at 1 2.15>]}}})
     Y=1
     Z=1
    ))
   &SolveForcesAndVelocity(Clamp Velocity=false, Limit Acceleration=false)
   &ScaleColor(
    Scale Alpha=&MergeHelper_FloatSwitch(
     0=&FloatFromCurve(
      FloatCurve=DataInterface() {propertyValues {Curve {keys [
          {arriveTangent=4.4826, interpMode=RCIM_Cubic, leaveTangent=4.4826, tangentMode=RCTM_User, time=0, value=0.2}
          {arriveTangent=-0.2, interpMode=RCIM_Cubic, leaveTangent=-0.2, time=0.3, value=1}
          <at 1 0>
         ]}}}
      Scale Curve=0.75
     )
     1=&FloatFromCurve(
      FloatCurve=DataInterface() {propertyValues {Curve {keys [
          {arriveTangent=4.4826, interpMode=RCIM_Cubic, leaveTangent=4.4826, tangentMode=RCTM_User, time=0, value=0.2}
          {arriveTangent=-0.2, interpMode=RCIM_Cubic, leaveTangent=-0.2, time=0.3, value=1}
          {arriveTangent=-0.8897, tangentMode=RCTM_User, time=1, value=0}
         ]}}}
      Scale Curve=2.5
     )
     2=?
     3=?
     Index=$System.VariantInt
     Num=::NewEnumerator0
    )
    Scale RGB=&MergeHelper_VectorSwitch(
     0=&MakeVectorFromLinearColorRGB(Color=&ColorFromCurve(ColorCurve=DataInterface() {propertyValues {
         AlphaCurve {keys [<at 0 1 RCIM_Cubic>]}
         BlueCurve {keys [<at 0 1.1 RCIM_Cubic>,<at 0.5031 0.7435>,<at 1 0.862 RCIM_Cubic>]}
         GreenCurve {keys [<at 0 1.1 RCIM_Cubic>,<at 0.5031 0.7435>,<at 1 0.862 RCIM_Cubic>]}
         RedCurve {keys [<at 0 1.1 RCIM_Cubic>,<at 0.5031 0.7435>,<at 1 0.862 RCIM_Cubic>]}
        }}))
     1=&MakeVectorFromLinearColorRGB(Color=&ColorFromCurve(ColorCurve=DataInterface() {propertyValues {
         AlphaCurve {keys [<at 0 1 RCIM_Cubic>]}
         BlueCurve {keys [<at 0 0.1589>,<at 0.477 0.0595>,<at 0.9907 0.0595>]}
         GreenCurve {keys [<at 0 0.1117>,<at 0.477 0.0185>,<at 0.9907 0.0185>]}
         RedCurve {keys [<at 0 0.1026>,<at 0.477 0.0123>,<at 0.9907 0.0123>]}
        }}))
     2=?
     3=?
     Index=$System.VariantInt
     Num=::NewEnumerator0
    )
    ScaleA=true
    ScaleRGB=true
   )
   &DynamicMaterialParameters(
    Index 0 Param 1=&FloatFromCurve(FloatCurve=DataInterface() {propertyValues {Curve {keys [<at 0 0.4>,<at 1 0.8>]}}})
    Index 0 Param 2=&FloatFromCurve(FloatCurve=DataInterface() {propertyValues {Curve {keys [
         {arriveTangent=-2.2616, interpMode=RCIM_Cubic, leaveTangent=-3.3361, tangentMode=RCTM_Break, time=0, value=1}
         {arriveTangent=-0.8527, interpMode=RCIM_Cubic, leaveTangent=-0.8527, tangentMode=RCTM_User, time=0.8, value=0}
        ]}}})
    Index 0 Param 3=0.5
    Module.Param0WriteEnabled=true
    Module.Param1WriteEnabled=true
    Module.Param2WriteEnabled=true
    Module.Param3WriteEnabled=true
   ) {module=DynamicMaterialParameters002}
  ]
  rendererClasses [&NiagaraMeshRendererProperties]
  renderers [{rendererClass=&NiagaraMeshRendererProperties, rendererIndex=0}]
 }
 Airflow_Stick {
  emitterUpdateScript [
   &EmitterState(Life Cycle Mode=::NewEnumerator1, Loop Behavior=::NewEnumerator1, Loop Delay=0.07, Loop Duration=0.5, UseLoopDelay=true)
   &SpawnBurst_Instantaneous(Use Spawn Probability=false, UseLoopCountLimit=false)
  ]
  particleSpawnScript [
   &InitializeParticle(Color=$System.TOD Color, Lifetime=0.25, Mesh Scale Mode=::NewEnumerator4, UsePositionOffset=false, Write Material Random=true)
   &AddVelocity(Velocity=<0 0 -100>)
   &InitialMeshOrientation(Mesh Orientation Mode=::NewEnumerator2, Rotation=<0 -0.25 0>, Use Rotation Vector=true)
   &SkeletalMeshLocation(
    Bone Sampling Mode=::NewEnumerator4
    MaskSpawnGroup=false
    Orientation Sampling=::NewEnumerator1
    Sampled Position Offset=<0 0 120>
    Skeletal Mesh=$User.Skeletal Mesh
   )
   &CameraOffset(Camera Offset Amount=30)
  ]
  particleUpdateScript [
   &ParticleState()
   &ScaleMeshSize(Scale Factor=&MakeVector(
     X=&FloatFromCurve(FloatCurve=DataInterface() {propertyValues {Curve {keys [<at 0 0.5>,<at 1 1>]}}}, Scale Curve=2.75)
     Y=1
     Z=1
    ))
   &SolveForcesAndVelocity(Clamp Velocity=false, Limit Acceleration=false)
   &DynamicMaterialParameters(
    Index 0 Param 1=&FloatFromCurve(FloatCurve=DataInterface() {propertyValues {Curve {keys [<at 0 0>,<at 0.5 0>,<at 1 0.08>]}}})
    Index 0 Param 2=&FloatFromCurve(FloatCurve=DataInterface() {propertyValues {Curve {keys [{arriveTangent=-2.2616, interpMode=RCIM_Cubic, time=0, value=-0.5}, {interpMode=RCIM_Cubic, leaveTangent=-0.7156, time=1, value=1}]}}})
    Index 0 Param 3=1.2
    Index 0 Param 4=&FloatFromCurve(
     FloatCurve=DataInterface() {propertyValues {Curve {keys [{arriveTangent=1.2782, interpMode=RCIM_Cubic, leaveTangent=1.2782, tangentMode=RCTM_User, time=0, value=1}, <at 1 2 RCIM_Cubic>]}}}
     Scale Curve=0.4
    )
    Module.Param0WriteEnabled=true
    Module.Param1WriteEnabled=true
    Module.Param2WriteEnabled=true
    Module.Param3WriteEnabled=true
   )
   &ScaleColor(
    Scale Alpha=&MergeHelper_FloatSwitch(
     0=&FloatFromCurve(FloatCurve=DataInterface() {propertyValues {Curve {keys [<at 0 1 RCIM_Cubic>, {arriveTangent=-1.8584, tangentMode=RCTM_User, time=1, value=0}]}}})
     1=&FloatFromCurve(
      FloatCurve=DataInterface() {propertyValues {Curve {keys [<at 0 1 RCIM_Cubic>, {arriveTangent=-2.7838, tangentMode=RCTM_User, time=1, value=0}]}}}
      Scale Curve=1.5
     )
     2=?
     3=?
     Index=$System.VariantInt
     Num=::NewEnumerator0
    )
    Scale RGB=&MergeHelper_VectorSwitch(
     0=&MakeVectorFromLinearColorRGB(Color=&ColorFromCurve(ColorCurve=DataInterface() {propertyValues {
         AlphaCurve {keys [<at 0 1 RCIM_Cubic>]}
         BlueCurve {keys [<at 0 1.7182 RCIM_Cubic>,<at 1 1 RCIM_Cubic>]}
         GreenCurve {keys [<at 0 2 RCIM_Cubic>,<at 1 1 RCIM_Cubic>]}
         RedCurve {keys [<at 0 1.9881 RCIM_Cubic>,<at 1 1 RCIM_Cubic>]}
        }}))
     1=&MakeVectorFromLinearColorRGB(Color=&ColorFromCurve(ColorCurve=DataInterface() {propertyValues {
         AlphaCurve {keys [<at 0 1 RCIM_Cubic>]}
         BlueCurve {keys [<at 0 0.6172 RCIM_Cubic>,<at 1 0.1166>]}
         GreenCurve {keys [<at 0 0.4264 RCIM_Cubic>,<at 1 0.0602>]}
         RedCurve {keys [<at 0 0.3894 RCIM_Cubic>,<at 1 0.0509>]}
        }}))
     2=?
     3=?
     Index=$System.VariantInt
     Num=::NewEnumerator0
    )
    ScaleA=true
    ScaleRGB=true
   )
  ]
  rendererClasses [&NiagaraMeshRendererProperties]
  renderers [
   {rendererClass=&NiagaraMeshRendererProperties, rendererIndex=0}
   {rendererClass=&NiagaraMeshRendererProperties, rendererIndex=1}
  ]
 }
 Sparks {
  bEnabled=false
  emitterUpdateScript [
   &EmitterState(Life Cycle Mode=::NewEnumerator1, Loop Behavior=::NewEnumerator1, Loop Delay=0.05, Loop Duration=0.5, UseLoopDelay=true)
   &SpawnBurst_Instantaneous(Spawn Count=10, Use Spawn Probability=false, UseLoopCountLimit=false)
  ]
  particleSpawnScript [
   &InitializeParticle(
    Lifetime Max=0.4
    Lifetime Min=0.3
    Lifetime Mode=::NewEnumerator1
    Sprite Size Max=<3 5>
    Sprite Size Min=<1 2>
    Sprite Size Mode=::NewEnumerator2
    UsePositionOffset=false
    Write Material Random=true
   )
   &SkeletalMeshLocation(Bone Sampling Mode=::NewEnumerator4, MaskSpawnGroup=false, Sampled Position Offset=<0 -3 20>, Skeletal Mesh=$User.Skeletal Mesh)
   &AddVelocity(Velocity=<0 0 -500>, Velocity Speed Scale=&RandomRangeFloat(Minimum=0.75, Maximum=1))
   &ShapeLocation(Box Size=<10 10 100>, MaskSpawnGroup=false, Override Seed=false, Shape Primitive=::NewEnumerator2, Surface Only=false, UseOwnerScale=true)
   &CameraOffset(Camera Offset Amount=10)
  ]
  particleUpdateScript [
   &ParticleState()
   &ScaleColor(
    Linear Color Curve=DataInterface() {propertyValues {
      AlphaCurve {keys [<at 0.333 1>,<at 1 0>]}
      BlueCurve {keys [<at 0 2.6235>,<at 0.1444 1.5>,<at 0.5925 3>,<at 1 1>]}
      GreenCurve {keys [<at 0 3>,<at 0.1444 0.9493>,<at 0.5925 0.8467>,<at 1 0>]}
      RedCurve {keys [<at 0 0>,<at 0.1444 0>,<at 0.5925 0>,<at 1 0.052>]}
     }}
    Scale Mode=2
   ) {module=ScaleColor001}
   &ScaleSpriteSize(Scale Factor=&Vector2DFromCurve(
     Vector2Curve=DataInterface() {propertyValues {
       XCurve {keys [<at 0 1 RCIM_Cubic>, {arriveTangent=-2.456, interpMode=RCIM_Cubic, leaveTangent=-2.456, tangentMode=RCTM_User, time=1, value=0}]}
       YCurve {keys [<at 0 1 RCIM_Cubic>, {arriveTangent=-2.456, interpMode=RCIM_Cubic, leaveTangent=-2.456, tangentMode=RCTM_User, time=1, value=0}]}
      }}
     ScaleCurve=<1 1>
    )) {module=ScaleSpriteSize001}
   /SetVariables(Particles.Glow03_Color=<rgba 0 0.3004 1 0.3>) {isSetParameters=true}
   /SetVariables(Particles.Glow03_Size=&Multiply_Vector2D(
     A=$Particles.SpriteSize
     B=&Multiply_Vector2D_ByFloat(
      Float=3
      Vector2D=&MakeVector2D(
       X=1
       Y=&FloatFromCurve(
        FloatCurve=DataInterface() {propertyValues {Curve {keys [<at 0 1 RCIM_Cubic>,<at 0.3 3.5 RCIM_Cubic>,<at 1 1 RCIM_Cubic>]}}}
        Scale Curve=&RandomRangeFloat(Minimum=0.7, Maximum=1)
       )
      )
     )
    )) {isSetParameters=true}
   &CurlNoiseForce(
    Noise Frequency=300
    Noise Strength=&Multiply_Float(A=2000, B=1)
    Pan Noise=true
    Pan Noise Field=<0 0 1>
    Random Seed=10
   ) {enabled=false}
   &Drag(
    Drag=&FloatFromCurve(FloatCurve=DataInterface() {propertyValues {Curve {keys [
         {interpMode=RCIM_Cubic, leaveTangent=12.0973, tangentMode=RCTM_Break, time=0, value=1}
         {interpMode=RCIM_Cubic, tangentMode=RCTM_User, time=1, value=6}
        ]}}})
    Use Linear Drag=true
    Use Rotational Drag=true
   )
   &SolveForcesAndVelocity(Clamp Velocity=false, Limit Acceleration=false)
   &ScaleSpriteSizeBySpeed(Max Scale Factor=<0.7 2>)
  ]
  rendererClasses [&NiagaraSpriteRendererProperties]
  renderers [
   {rendererClass=&NiagaraSpriteRendererProperties}
   {rendererClass=&NiagaraSpriteRendererProperties, rendererIndex=1}
  ]
 }
}
summary {
 systemName=NS_Pickaxe_MicStand_CreamDecoy_FirstAttack
 emitters [
  {emitterName=Airflow_Stick, rendererClasses [&NiagaraMeshRendererProperties]}
  {emitterName=AirflowColor, rendererClasses [&NiagaraMeshRendererProperties]}
  {emitterName=Air_cone, rendererClasses [&NiagaraMeshRendererProperties]}
  {bEnabled=false, emitterName=Sparks, rendererClasses [&NiagaraSpriteRendererProperties]}
 ]
 userVariables [
  User.HitColor {type=LinearColor, description="", defaultValue {struct=LinearColor, value=<rgba 0.2634 0.9002 1 1>}}
  User.Skeletal Mesh {
   type=NiagaraDataInterfaceSkeletalMesh
   defaultValue {
    struct=NiagaraExt_VariableValue_DataInterface
    value {dataInterface=&self:NiagaraDataInterfaceSkeletalMesh_0}
   }
  }
  User.VariantID {type=NiagaraFloat, defaultValue=0, description=""}
 ]
}
